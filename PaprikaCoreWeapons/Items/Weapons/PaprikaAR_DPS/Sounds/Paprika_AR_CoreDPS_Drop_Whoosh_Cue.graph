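# Sound cue that plays the drop whoosh for the Paprika Core DPS assault rifle.
# A single wave player for the drop sound feeds a Modulator left at default settings as the root.
# The cue uses the shared Fort weapon drop attenuation and weapon foley sound class at reduced volume.
# It overrides concurrency and has its debug flag switched on.
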SoundCue /PaprikaCoreWeapons/Items/Weapons/PaprikaAR_DPS/Sounds/Paprika_AR_CoreDPS_Drop_Whoosh_Cue
root=Modulator_1
nodes [
 {id=Modulator_1, children [WavePlayer_1]}
 {id=WavePlayer_1, properties {soundWaveAssetPtr=&/Paprika_AR_CoreDPS_Drop}}
]
properties {
 attenuationSettings=&Fort_Weapon_Drop_Att
 bDebug=true
 bOverrideConcurrency=true
 maxDistance=3000
 soundClassObject=&Fort_WeaponFoley
 volumeMultiplier=0.451
}
