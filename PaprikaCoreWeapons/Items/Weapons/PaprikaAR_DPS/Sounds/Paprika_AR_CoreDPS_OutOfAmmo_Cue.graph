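# Sound cue for the out-of-ammo sound of the Paprika core DPS assault rifle.
# A single first-person out-of-ammo wave plays through a Modulator node left at its default settings.
# The cue uses weapon reload attenuation and the weapon foley sound class, with a heavily boosted volume multiplier.
# It overrides concurrency and has its debug flag enabled.
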SoundCue /PaprikaCoreWeapons/Items/Weapons/PaprikaAR_DPS/Sounds/Paprika_AR_CoreDPS_OutOfAmmo_Cue
root=Modulator_0
nodes [
 {id=Modulator_0, children [WavePlayer_0]}
 {
  id=WavePlayer_0
  properties {soundWaveAssetPtr=&/Paprika_AR_CoreDPS_Fire_1P_OutOfAmmo_01}
 }
]
properties {
 attenuationSettings=&Fort_Weapon_Reload_Att
 bDebug=true
 bOverrideConcurrency=true
 maxDistance=2250
 soundClassObject=&Fort_WeaponFoley
 volumeMultiplier=5.012
}
