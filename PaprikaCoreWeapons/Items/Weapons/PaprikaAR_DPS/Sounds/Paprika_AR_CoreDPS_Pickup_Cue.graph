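# Sound cue that plays the pickup sound for the Paprika CoreDPS assault rifle.
# A Random node picks between two equally weighted pickup waves, then feeds a Modulator for playback variation.
# It uses the shared weapon pickup attenuation and weapon foley sound class at reduced volume, overrides concurrency, and has its debug flag left enabled.
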SoundCue /PaprikaCoreWeapons/Items/Weapons/PaprikaAR_DPS/Sounds/Paprika_AR_CoreDPS_Pickup_Cue
root=Modulator_0
nodes [
 {id=Modulator_0, children [Random_0]}
 {id=Random_0, children [WavePlayer_0,WavePlayer_1], properties {weights [1,1]}}
 {id=WavePlayer_0, properties {soundWaveAssetPtr=&/Paprika_AR_CoreDPS_Pickup_01}}
 {id=WavePlayer_1, properties {soundWaveAssetPtr=&/Paprika_AR_CoreDPS_Pickup_02}}
]
properties {
 attenuationSettings=&Fort_Weapon_Pickup_Att
 bDebug=true
 bOverrideConcurrency=true
 maxDistance=3500
 soundClassObject=&Fort_WeaponFoley
 volumeMultiplier=0.6
}
