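# Sound cue for the end stage of the Paprika CoreDPS assault rifle's reload.
# A Random node picks between two equally weighted reload-end waves, and a Modulator node at the root sits over the result.
# It plays in the weapon foley sound class with weapon reload attenuation and reduced volume.
# The cue overrides concurrency and has its debug flag left on.
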SoundCue /PaprikaCoreWeapons/Items/Weapons/PaprikaAR_DPS/Sounds/Paprika_AR_CoreDPS_Reload_End_Cue
root=Modulator_0
nodes [
 {id=Modulator_0, children [Random_0]}
 {id=Random_0, children [WavePlayer_2,WavePlayer_3], properties {weights [1,1]}}
 {id=WavePlayer_2, properties {soundWaveAssetPtr=&/Paprika_AR_CoreDPS_Reload_End_01}}
 {id=WavePlayer_3, properties {soundWaveAssetPtr=&/Paprika_AR_CoreDPS_Reload_End_02}}
]
properties {
 attenuationSettings=&Fort_Weapon_Reload_Att
 bDebug=true
 bOverrideConcurrency=true
 maxDistance=2250
 soundClassObject=&Fort_WeaponFoley
 volumeMultiplier=0.451
}
